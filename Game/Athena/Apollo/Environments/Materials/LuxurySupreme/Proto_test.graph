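# Prototype LuxurySupreme environment material for Apollo: a parameterized metal surface with emissive that gains dust and blends into the landscape near the ground.
# World Z minus runtime virtual texture height drives separate distance masks for dust, landscape normals and landscape surface attributes.
# Attributes are converted to world-space normals before blending, behind RVT switch functions.
# Forward, mobile and material proxy paths skip the landscape blend.
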Material /Game/Athena/Apollo/Environments/Materials/LuxurySupreme/Proto_test
graph {
 properties {bTangentSpaceNormal=false, bUseMaterialAttributes=true, pixelDepthOffsetMode=PDOM_Legacy, usedShadingModels=Default Lit}
 nodes [
  n1 = RuntimeVirtualTextureSampleParameter() {materialType=WorldHeight, parameterName=RVTHeight, virtualTexture=/Game/Athena/Apollo/Maps/Landscape/VT/RVT_Apollo_Landscape_Height}
  n2 = Subtract(A=ComponentMask(None=WorldPosition().XYZ) {b=true}, B=@n1.WorldHeight)
  n3 = Saturate(None=Divide(A=@n2, B=ScalarParameter() {defaultValue=2000, group=RVT, parameterName=Dust RVT}))
  NamedRerouteDeclaration(None=@n3) {name=DustMask, nodeColor=<rgba 0 1 0.25 1>}
  Reroute(None=@n3)
  n4 = SetMaterialAttributes!(
   MaterialAttributes
   Base Color=VectorParameter() {group=Dust, parameterName=DustColor, defaultValue=<rgba 0.0286 0.0189 0.0134 0>}.RGB
   Metallic=Constant()
   Specular=Constant() {r=0.4}
   Roughness=Constant() {r=1}
  ) {attributeSetTypes [
    #1
    #2
    #3
    #4
   ]}
  n5 = SetMaterialAttributes!(
   MaterialAttributes
   Base Color=VectorParameter() {parameterName=Color, defaultValue=<rgba 0.5208 0.5208 0.5208 1>}.RGB
   Metallic=ScalarParameter() {defaultValue=1, parameterName=Metal}
   Specular=ScalarParameter() {defaultValue=0.5, parameterName=Specular}
   Roughness=ScalarParameter() {defaultValue=0.25, parameterName=Roughness}
   Emissive Color=Multiply(A=ScalarParameter() {parameterName=EmissivePower}, B=VectorParameter() {parameterName=EmissiveColor, defaultValue=<rgba 0 0 0 1>}.RGB)
  ) {attributeSetTypes [
    #1
    #2
    #3
    #4
    #5
   ]}
  n6 = Constant() {r=1}
  n7 = NamedRerouteUsage()
  n8 = MaterialFunctionCall!(Material=BlendMaterialAttributes(
    A=@n4
    B=@n5
    Alpha=MaterialFunctionCall!(NoRVT=@n6, RVT=@n7) {materialFunction=/Game/Packages/Fortress_SharedMaterials/Base_Material_Functions/FN_RVT_Switch_Color}.Result
   )) {materialFunction=/Game/Athena/Apollo/Maps/Landscape/VT/MaterialFunctions/TransformTangentToWorld}
  n9 = RuntimeVirtualTextureSampleParameter() {materialType=BaseColor_Normal_Specular_YCoCg, parameterName=RVT, virtualTexture=/Game/Athena/Apollo/Maps/Landscape/VT/RVT_Apollo_Landscape}
  n10 = NamedRerouteDeclaration(None=@n2) {name=RVT Height, nodeColor=<rgba 1 0 0.1406 1>}
  n11 = ScalarParameter() {defaultValue=155, group=RVT, parameterName=Blend RVT Normals}
  n12 = BlendMaterialAttributes(
   A=SetMaterialAttributes!(MaterialAttributes=@n8.Result, Normal=@n9.Normal) {attributeSetTypes [#6]}
   B=@n8.Result
   Alpha=Saturate(None=Divide(A=@n10, B=@n11))
  ) {vertexAttributeBlendType=UseA}
  n13 = SetMaterialAttributes!(MaterialAttributes=@n12, Base Color=@n9.BaseColor, Specular=@n9.BaseColor, Roughness=@n9.BaseColor, Metallic=Constant()) {attributeSetTypes [
    #1
    #3
    #4
    #2
   ]}
  n14 = Saturate(None=Divide(A=@n2, B=ScalarParameter() {defaultValue=555, group=RVT, parameterName=Blend RVT}))
  n15 = MaterialFunctionCall(
   NoRVT=@n8.Result
   RVT=MaterialProxyReplace(Realtime=BlendMaterialAttributes(A=@n13, B=@n12, Alpha=@n14) {vertexAttributeBlendType=UseA}, MaterialProxy=@n8.Result)
  ) {materialFunction=&FN_RVT_Switch_Material}
  ShadingPathSwitch(Default=@n15.Result, Deferred=@n15.Result, Forward=@n8.Result, Mobile=@n8.Result)
 ]
}
